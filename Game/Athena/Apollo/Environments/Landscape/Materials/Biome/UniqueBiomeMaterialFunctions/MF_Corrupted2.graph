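# Landscape biome material function that tints incoming material attributes with corrupted purple tones and outputs a Height mask.
# The default path samples world-XY noise, warped by a storm swirl mask and BaseColor, to mix purples over the original color.
# Forward and Mobile use one warped mask texture instead and flatten specular, roughness and normal.
# Most C2 and RedoColor parameters are unconnected, with constants used in their place.
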MaterialFunction /Game/Athena/Apollo/Environments/Landscape/Materials/Biome/UniqueBiomeMaterialFunctions/MF_Corrupted2
graph {nodes [
  Constant() {r=0.74}
  Constant() {r=80000}
  n1 = FunctionInput() {inputName=Material, inputType=FunctionInput_MaterialAttributes}
  n2 = GetMaterialAttributes(None=@n1) {attributeGetTypes [#1]}
  n3 = WorldPosition()
  n4 = TextureSample(UVs=Divide(A=ComponentMask(None=@n3.XYZ) {g=true, r=true}, B=Constant() {r=3500})) {
   samplerSource=SSM_Wrap_WorldGroupSettings
   samplerType=SAMPLERTYPE_LinearGrayscale
   texture=/Game/Global/Textures/Default/Flowmaps/T_StormSwirlMask
  }
  n5 = Multiply(
   A=ConstantBiasScale(None=@n4.RGB) {bias=-0.5}
   B=VectorParameter() {group=WPO, parameterName="C2-UV=MoveDist", defaultValue=<rgba -0.02 0.02 0 1>}.RGB
  )
  n6 = Multiply(A=ComponentMask(None=@n2.BaseColor) {g=true, r=true}, B=Constant() {r=0.261})
  n7 = Divide(A=ComponentMask(None=WorldPosition().XYZ) {g=true, r=true}, B=ScalarParameter() {defaultValue=80000, parameterName=C2-NoiseScale})
  n8 = TextureSample(UVs=Add(A=ComponentMask(None=@n5) {g=true, r=true}, B=Add(A=@n6, B=@n7))) {samplerSource=SSM_Wrap_WorldGroupSettings, samplerType=SAMPLERTYPE_LinearColor, texture=/Game/Global/Textures/Default/Noise/T_Apollo_TCNoise}
  n9 = LinearInterpolate(A=Constant3Vector() {constant=<rgba 0.0302 0.004 0.08 1>}, B=Constant3Vector() {constant=<rgba 0.1933 0.05 0.25 1>}, Alpha=@n8.R)
  n10 = Constant3Vector() {constant=<rgba 0.0982 0.0089 0.2448 1>}
  n11 = Add(A=Constant() {r=0.2}, B=Multiply(A=Constant() {r=2}, B=@n8.R))
  n12 = LinearInterpolate(
   A=@n2.BaseColor
   B=LinearInterpolate(A=LinearInterpolate(A=@n9, B=@n10, Alpha=@n8.G), B=Constant3Vector() {constant=<rgba 0.0655 0.047 0.3 1>}, Alpha=@n8.B)
   Alpha=LinearInterpolate(A=Constant() {r=1}, B=Saturate(None=@n11), Alpha=Constant() {r=0.5})
  )
  n13 = Constant3Vector() {constant=<rgba 0.1933 0.05 0.25 1>}
  n14 = ComponentMask(None=WorldPosition().XYZ) {g=true, r=true}
  n15 = Constant() {r=10000}
  n16 = GetMaterialAttributes(None=@n1) {attributeGetTypes [#1]}
  n17 = ComponentMask(None=@n16.BaseColor) {g=true, r=true}
  n18 = Constant() {r=0.75}
  n19 = TextureSample(UVs=Add(A=Divide(A=@n14, B=@n15), B=Multiply(A=@n17, B=@n18))) {
   samplerSource=SSM_Wrap_WorldGroupSettings
   samplerType=SAMPLERTYPE_LinearColor
   texture=/Game/Athena/Environments/Landscape/Textures/T_Athena_Snow_01_MOBILE_M
  }
  n20 = LinearInterpolate(A=Constant3Vector() {constant=<rgba 0.0437 0.0313 0.2 1>}, B=Constant3Vector() {constant=<rgba 0.08 0 0.046 1>}, Alpha=@n19.G)
  n21 = Saturate(None=LinearInterpolate(A=Constant() {r=-0.25}, B=Constant() {r=3}, Alpha=@n19.A))
  n22 = LinearInterpolate(A=Constant() {r=-1}, B=Constant() {r=2}, Alpha=@n19.A)
  n23 = SetMaterialAttributes!(
   MaterialAttributes
   BaseColor=LinearInterpolate(A=LinearInterpolate(A=@n13, B=@n20, Alpha=@n21), B=@n16.BaseColor, Alpha=Saturate(None=@n22))
   Specular=Constant() {r=0.3}
   Roughness=Constant() {r=0.9}
   Normal=VertexNormalWS()
  ) {attributeSetTypes [
    #1
    #2
    #3
    #4
   ]}
  FunctionOutput(None=ShadingPathSwitch(
    Default=SetMaterialAttributes!(MaterialAttributes=@n2.MaterialAttributes, BaseColor=@n12) {attributeSetTypes [#1]}
    Forward=@n23
    Mobile=@n23
   )) {outputName=Material}
  FunctionOutput(None=ShadingPathSwitch(Default=@n8.R, Forward=@n19.A, Mobile=@n19.A)) {outputName=Height}
  LinearInterpolate(
   A=Constant() {r=0.15}
   B=Constant() {r=0.3}
   Alpha=MaterialFunctionCall(Min=Constant() {r=1}, Max=Constant() {r=0.25}) {materialFunction=&SmoothStep}.Result
  )
  Multiply(A=@n5, B=ScalarParameter() {defaultValue=1, parameterName=C2-WarpMod})
  ScalarParameter() {defaultValue=10000, parameterName=C2-Mobile-WarpScale}
  ScalarParameter() {defaultValue=-0.25, parameterName=C2-Mobile-ColorMaskA}
  ScalarParameter() {defaultValue=0.5, parameterName=C2-SoftBlend}
  ScalarParameter() {defaultValue=0.3, parameterName=C2-HLerpB}
  ScalarParameter() {defaultValue=3500, parameterName=C2-WarpScale}
  ScalarParameter() {defaultValue=0.75, parameterName=C2-Mobile-DiffuseWarpAmt}
  ScalarParameter() {defaultValue=0.25, parameterName=C2-SSMax}
  ScalarParameter() {defaultValue=1, parameterName=C2-SSMin}
  ScalarParameter() {defaultValue=-1, parameterName=C2-Mobile-HeightA}
  ScalarParameter() {defaultValue=2, parameterName=C2-BlendMaskMod}
  ScalarParameter() {defaultValue=0.2, parameterName=C2-BlendMaskAdd}
  ScalarParameter() {defaultValue=0.15, parameterName=C2-HLerpA}
  ScalarParameter() {defaultValue=2, parameterName=C2-Mobile-HeightB}
  ScalarParameter() {defaultValue=3, parameterName=C2-Mobile-ColorMaskB}
  VectorParameter() {parameterName=RedoColor1-Mobile, defaultValue=<rgba 0.1933 0.05 0.25 1>}
  VectorParameter() {parameterName=RedoColor1, defaultValue=<rgba 0.1933 0.05 0.25 1>}
  VectorParameter() {parameterName=RedoColor0-Mobile, defaultValue=<rgba 0.08 0 0.046 1>}
  VectorParameter() {parameterName=RedoColor2, defaultValue=<rgba 0.0982 0.0089 0.2448 1>}
  VectorParameter() {parameterName=RedoColor3, defaultValue=<rgba 0.0655 0.047 0.3 1>}
  VectorParameter() {parameterName=RedoColor0, defaultValue=<rgba 0.0302 0.004 0.08 1>}
  VectorParameter() {parameterName=RedoColor2-Mobile, defaultValue=<rgba 0.0437 0.0313 0.2 1>}
 ]}
