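# Material function that builds Apollo landscape snow progression masks from a stage mask and an area detail mask.
# Global SnowMaskOffset X/Y parameters shift roughly one-third-scaled UVs to select a stage tile, apparently from a 3x3 phase atlas.
# It outputs stage blue, detail green and blue, and stage times detail green for render targets.
# A static switch picks mesh or terrain UVs; both UseMinimapCoords branches use identical coordinates.
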MaterialFunction /Game/Athena/Apollo/Environments/Landscape/Materials/Biome/UniqueBiomeMaterialFunctions/MF_Snow_Progression_Mask
graph {nodes [
  Constant() {r=0.668}
  Constant() {desc=Phase 7}
  Constant() {r=0.668}
  Constant() {desc=Phase 4}
  Constant() {r=0.334}
  Constant() {desc=Phase 5, r=0.334}
  Constant() {r=0.334}
  Constant() {desc=Phase 6, r=0.668}
  Constant() {r=0.334}
  Constant() {desc=Phase 1}
  Constant()
  Constant() {desc=Phase 2, r=0.334}
  Constant()
  Constant() {desc=Phase 3, r=0.668}
  Constant()
  Constant() {desc="^Tilted\r\nexposed"}
  Constant() {desc=Phase 9, r=0.668}
  Constant() {r=0.668}
  Constant() {desc=Phase 8, r=0.334}
  n1 = CollectionParameter() {collection=&FortniteMaterialParameters, parameterName=SnowMaskOffsetX}
  n2 = CollectionParameter() {collection=&FortniteMaterialParameters, parameterName=SnowMaskOffsetY}
  n3 = Constant() {r=0.332}
  n4 = NamedRerouteUsage()
  n5 = FunctionInput(Preview=TextureObject() {samplerType=SAMPLERTYPE_Normal, texture=&FlatNormalBlue}) {bUsePreviewValueAsDefault=true, description=T_Athena_Snow_01_MOBILE_M, inputName=SnowStageMask, inputType=FunctionInput_Texture2D}
  n6 = TextureSample(
   UVs=Add(
    A=MaterialFunctionCall(X=@n1, Y=@n2) {materialFunction=&MakeFloat2}.Result
    B=Multiply(A=@n3, B=@n4)
   )
   Tex=@n5
  ) {
   desc=T_Athena_Snow_01_MOBILE_M
   samplerSource=SSM_Wrap_WorldGroupSettings
   samplerType=SAMPLERTYPE_Normal
   texture=&FlatNormalBlue
  }
  FunctionOutput(None=@n6.B) {outputName=Snow-SUB-UV-Mask}
  n7 = FunctionInput(Preview=TextureObject() {samplerType=SAMPLERTYPE_Normal, texture=&FlatNormalBlue}) {
   bUsePreviewValueAsDefault=true
   description=T_SnowMask_Progression
   inputName=SnowAreaDetailMask
   inputType=FunctionInput_Texture2D
   sortPriority=1
  }
  n8 = TextureSample(UVs=NamedRerouteUsage(), Tex=@n7) {
   desc=T_SnowMask_Progression
   samplerSource=SSM_Wrap_WorldGroupSettings
   samplerType=SAMPLERTYPE_Normal
   texture=&FlatNormalBlue
  }
  FunctionOutput(None=@n8.G) {outputName=SnowAreaDetail-G, sortPriority=1}
  FunctionOutput(None=@n8.B) {outputName=SnowAreaDetail-B, sortPriority=2}
  FunctionOutput(None=Multiply(A=@n6.B, B=@n8.G)) {outputName=ForRenderTarget, sortPriority=4}
  n9 = MaterialFunctionCall!(WorldPosition) {materialFunction=/Game/Athena/Environments/Landscape/MaterialFunctions/Utilities/MF_Apollo_Terrain_Coords}
  n10 = FunctionInput(Preview=StaticBool()) {bUsePreviewValueAsDefault=true, description="default=false", inputName=UseMinimapCoords, inputType=FunctionInput_StaticBool, sortPriority=2}
  n11 = FunctionInput(Preview=StaticBool()) {bUsePreviewValueAsDefault=true, description="default=false", inputName=IsForRenderTarget, inputType=FunctionInput_StaticBool, sortPriority=3}
  NamedRerouteDeclaration(None=StaticSwitch(True=TextureCoordinate(), False=StaticSwitch(True=@n9.MatchForMinimap, False=@n9.MatchForMinimap, Value=@n10), Value=@n11)) {name=UV, nodeColor=<rgba 1 0 1 1>}
  ScalarParameter() {group=SnowProgression, parameterName=SPM-AddX}
  ScalarParameter() {group=SnowProgression, parameterName=SPM-AddY}
 ]}
