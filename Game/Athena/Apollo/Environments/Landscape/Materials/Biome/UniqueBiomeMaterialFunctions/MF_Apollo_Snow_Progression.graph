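# Apollo island material function that blends progressive or static mountain snow over input material attributes, with an alternate WPO path for pushing grass down.
# Snow coverage collection values, radial, height, occlusion and slope masks, and world-aligned noise build a height-lerped snow mask.
# The mask blends albedo, roughness, specular and optional normals, adding time-of-day-driven sparkle on high quality.
# Static switches gate features; extra outputs expose masks and albedo, including RVT and minimap variants.
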MaterialFunction /Game/Athena/Apollo/Environments/Landscape/Materials/Biome/UniqueBiomeMaterialFunctions/MF_Apollo_Snow_Progression
graph {nodes [
  n1 = Multiply(A=NamedRerouteUsage(), B=ScalarParameter() {defaultValue=-100, group=SnowProgression, parameterName=SnowPushGrassDown})
  n2 = GetMaterialAttributes(None=NamedRerouteUsage()) {attributeGetTypes [#1]}
  Add(A=Multiply(A=Constant3Vector() {constant=<rgba 0 0 1 0>}, B=@n1), B=@n2.WorldPositionOffset)
  Constant3Vector() {constant=<rgba -94848 84352 0 1>}
  Constant3Vector() {constant=<rgba 70000 -140000 0 1>}
  Constant() {r=8}
  Constant() {r=-100}
  Constant() {r=-0.1}
  Constant() {r=1.75}
  Constant() {r=15}
  Constant() {r=0.94}
  Constant() {r=0.1}
  Constant() {r=20}
  Constant()
  Constant() {r=4000}
  Constant() {desc="Min snow amount (tips of mountains)", r=0.33}
  Constant() {r=2000}
  Constant() {desc=OG, r=0.8}
  Constant() {desc=OG, r=0.75}
  Constant() {r=0.9}
  Constant() {r=0.75}
  Constant() {r=4}
  Constant() {desc=TPS-SM-Rad-High-Trees, r=85000}
  Constant() {r=-5000}
  Constant() {r=5000}
  Constant() {desc="Max snow amount (covers island)", r=0.75}
  Constant() {desc="Min snow amount (tips of mountains)", r=89600}
  Constant() {r=0.5}
  n3 = MaterialFunctionCall(Float3=@n2.WorldPositionOffset) {materialFunction=&BreakOutFloat3Components}
  n4 = Add(A=@n3.B, B=@n1)
  n5 = Constant() {r=0.35}
  n6 = ShadingPathSwitch(Default=LinearInterpolate(A=NamedRerouteUsage(), B=Constant() {r=0.7}, Alpha=NamedRerouteUsage()), Forward=@n5, Mobile=@n5)
  n7 = NamedRerouteUsage()
  n8 = Max(A=NamedRerouteUsage(), B=@n7)
  n9 = StaticSwitch(True=QualitySwitch(Default=@n7, Low=@n7, High=@n8, Medium=@n7, Epic=@n8), False=@n7, Value=NamedRerouteUsage())
  n10 = Constant() {r=0.5}
  n11 = LinearInterpolate(A=NamedRerouteUsage(), B=NamedRerouteUsage(), Alpha=NamedRerouteUsage())
  n12 = NamedRerouteUsage()
  n13 = FunctionInput(Preview=StaticBool()) {bUsePreviewValueAsDefault=true, description="Default = False", inputName=UseSnowNormals, inputType=FunctionInput_StaticBool, sortPriority=4}
  n14 = NamedRerouteUsage()
  n15 = SetMaterialAttributes!(
   MaterialAttributes=NamedRerouteUsage()
   BaseColor=LinearInterpolate(A=NamedRerouteUsage(), B=NamedRerouteDeclaration(None=@n6) {name=OutputSnowAlbedo, nodeColor=<rgba 1 0 1 1>}, Alpha=NamedRerouteUsage())
   Roughness=LinearInterpolate(A=NamedRerouteUsage(), B=ShadingPathSwitch(Default=@n9, Forward=@n10, Mobile=@n10), Alpha=NamedRerouteUsage())
   Specular=LinearInterpolate(A=NamedRerouteUsage(), B=Constant() {r=0.5}, Alpha=NamedRerouteUsage())
   Normal=ShadingPathSwitch(Default=StaticSwitch(True=@n11, False=@n12, Value=@n13), Forward=@n14, Mobile=@n14)
   EmissiveColor=StaticSwitch(True=NamedRerouteUsage(), False=Constant(), Value=NamedRerouteUsage())
  ) {attributeSetTypes [
    #2
    #3
    #4
    #5
    #6
   ]}
  n16 = FunctionInput(Preview=StaticBool()) {
   bUsePreviewValueAsDefault=true
   description="Default=False. If true, this will move the mesh on Z using the parameter SnowPushGrassDown"
   inputName=IsForGrassMeshToPushDown
   inputType=FunctionInput_StaticBool
  }
  n17 = StaticSwitch(
   True=SetMaterialAttributes!(
    MaterialAttributes=@n2.MaterialAttributes
    WorldPositionOffset=MaterialFunctionCall(X=@n3.R, Y=@n3.R, Z=@n4) {materialFunction=&MakeFloat3}.Result
   ) {attributeSetTypes [#1]}
   False=@n15
   Value=NamedRerouteDeclaration(None=@n16) {name=IsForGrassMeshToPushDown, nodeColor=<rgba 1 0 1 1>}
  )
  n18 = NamedRerouteUsage()
  n19 = StaticSwitchParameter(True=@n17, False=@n18) {defaultValue=true, group=SnowProgression, parameterName=Enable Progression}
  n20 = StaticBool() {desc=Master switch}
  n21 = NamedRerouteUsage()
  FunctionOutput(None=StaticSwitch(True=@n19, False=StaticSwitch(True=@n19, False=@n18, Value=@n20) {desc=Master switch}, Value=@n21)) {desc=Material Output, outputName=Material, sortPriority=200}
  FunctionOutput(None=NamedRerouteUsage()) {desc=Soft Mask, outputName=SimpleMask, sortPriority=80}
  FunctionOutput(None=NamedRerouteUsage()) {outputName=MinimapMask, sortPriority=90}
  FunctionOutput(None=NamedRerouteUsage()) {outputName=SnowAlbedo, sortPriority=100}
  FunctionOutput(None=NamedRerouteUsage()) {desc=Soft Mask, outputName=DetailMask, sortPriority=70}
  n22 = NamedRerouteUsage()
  n23 = StaticSwitchParameter(True=@n17, False=@n22) {
   desc=This should always be set to false in this MF so it can be controlled by specific MI's
   group=SnowProgression
   parameterName="Enable Progression-Default=False"
  }
  FunctionOutput(None=StaticSwitch(True=@n23, False=StaticSwitch(True=@n23, False=@n22, Value=@n20) {desc=Master switch}, Value=@n21)) {desc=Material Output, outputName="DefaultSwitch=FalseMaterial", sortPriority=300}
  FunctionOutput(None=NamedRerouteUsage()) {desc=Soft Mask, outputName=SnowMask, sortPriority=60}
  FunctionOutput(None=NamedRerouteUsage()) {outputName=SnowAlbedoForRVT, sortPriority=95}
  FunctionOutput(None=NamedRerouteUsage()) {desc=Soft Mask, outputName=SnowMaskForRVT, sortPriority=65}
  n24 = LandscapeLayerSample() {parameterName=BiomeD}
  LinearInterpolate(A=Constant() {r=1}, B=Constant(), Alpha=@n24)
  n25 = ScalarParameter() {defaultValue=-0.1, group=SnowProgression, parameterName=TPS-HLerpLow}
  LinearInterpolate(
   A=@n25
   B=FunctionInput(Preview=ScalarParameter() {defaultValue=1.75, group=SnowProgression, parameterName=TPS-HLerpHigh}) {bUsePreviewValueAsDefault=true, inputName=LerpHighM, inputType=FunctionInput_Scalar}
   Alpha=NamedRerouteUsage()
  )
  n26 = TextureCoordinate()
  n27 = Constant2Vector() {g=0.854, r=0.8175}
  n28 = ScalarParameter() {defaultValue=0.33, group=SnowProgression, parameterName=TPS-Rad-Minimap}
  n29 = NamedRerouteUsage()
  n30 = LinearInterpolate(
   A=ScalarParameter() {defaultValue=0.33, desc="Min snow amount (tips of mountains)", group=SnowProgression, parameterName=TPS-TransLerpLow}
   B=ScalarParameter() {defaultValue=0.75, desc="Max snow amount (covers island)", group=SnowProgression, parameterName=TPS-TransLerpHigh}
   Alpha=NamedRerouteUsage()
  )
  n31 = FunctionInput(Preview=Constant() {r=2}) {bUsePreviewValueAsDefault=true, inputName=TemporalHeightLerpContrast, inputType=FunctionInput_Scalar}
  MaterialFunctionCall(
   Transition Phase=Multiply(
    A=SphereMask(A=@n26, B=@n27, Radius=@n28, Hardness=@n29) {hardnessPercent=0}
    B=MaterialFunctionCall(Transition Phase=@n30, Height Texture=@n25, Contrast=@n31) {materialFunction=&HeightLerp}.Alpha
   )
   Height Texture=NamedRerouteUsage()
  ) {materialFunction=&HeightLerp}
  n32 = CollectionParameter() {collection=&FortniteMaterialParameters, parameterName=SnowCoverage}
  Max(A=@n32, B=ScalarParameter() {group=SnowProgression, parameterName=DEBUG-SnowCoverage})
  n33 = CollectionParameter() {collection=&FortniteMaterialParameters, parameterName=FallingSnowCoverage}
  Max(A=ScalarParameter() {parameterName=DEBUG-FallingSnowCoverage}, B=@n33)
  n34 = ScalarParameter() {group=SnowProgression, parameterName=DEBUG-Override-SphereMask}
  n35 = ScalarParameter() {defaultValue=190000, desc="Min snow amount (tips of mountains)", group=SnowProgression, parameterName=TPS-SM-Rad-Low}
  n36 = ScalarParameter() {defaultValue=85000, desc=TPS-SM-Rad-High-Trees, parameterName=TPS-SM-Rad-High}
  n37 = NamedRerouteUsage()
  n38 = ScalarParameter() {defaultValue=0.75, group=SnowProgression, parameterName=zSnowRadMask-Hardlow}
  n39 = ScalarParameter() {defaultValue=0.9, group=SnowProgression, parameterName=zSnowRadMask-Hard}
  n40 = NamedRerouteUsage()
  n41 = SphereMask(
   A=ComponentMask(None=WorldPosition() {worldPositionShaderOffset=WPT_ExcludeAllShaderOffsets}.XYZ) {g=true, r=true}
   B=ComponentMask(None=VectorParameter() {parameterName=TPS-SM-Loc, defaultValue=<rgba 70000 -140000 0 1>}.RGB) {g=true, r=true}
   Radius=StaticSwitch(True=LinearInterpolate(A=@n35, B=@n36, Alpha=@n37), False=@n35, Value=NamedRerouteUsage())
   Hardness=NamedRerouteDeclaration(None=LinearInterpolate(A=@n38, B=@n39, Alpha=@n40)) {name=RadMask-Hardness, nodeColor=<rgba 1 0 1 1>}
  ) {hardnessPercent=0}
  n42 = ScalarParameter() {group=SnowProgression, parameterName=DEBUG-Override-WorldHeightMask}
  n43 = ComponentMask(None=WorldPosition() {worldPositionShaderOffset=WPT_ExcludeAllShaderOffsets}.XYZ) {b=true}
  n44 = NamedRerouteUsage()
  n45 = Constant() {r=1.5}
  n46 = LinearInterpolate(
   A=ScalarParameter() {defaultValue=-20000, group=SnowProgression, parameterName=TPS-SnowBlendHeightOffset-Min}
   B=ScalarParameter() {defaultValue=5000, group=SnowProgression, parameterName=TPS-SnowBlendHeightOffset-Max}
   Alpha=StaticSwitch(True=Power(Base=@n44, Exp=@n45), False=@n44, Value=NamedRerouteUsage())
  )
  n47 = ScalarParameter() {defaultValue=4000, group=SnowProgression, parameterName=TPS-SnowBlendDistance-Min}
  n48 = ScalarParameter() {defaultValue=2000, group=SnowProgression, parameterName=TPS-SnowBlendDistance-Max}
  n49 = NamedRerouteUsage()
  n50 = Saturate(None=Divide(A=Add(A=@n43, B=@n46), B=LinearInterpolate(A=@n47, B=@n48, Alpha=@n49)))
  Multiply(A=Saturate(None=Add(A=@n34, B=@n41)), B=Saturate(None=Add(A=@n42, B=@n50)))
  n51 = LinearInterpolate(Alpha=MaterialFunctionCall() {materialFunction=&LinearGradient}.UGradient) {constA=-0.05, constB=1.05}
  n52 = Constant() {r=1.25}
  Multiply(A=MaterialFunctionCall!(
    X range Y range=Constant4Vector() {constant=<rgba 0 1 0 1>}.RGBA
    Gradient Height
    ShowGrid?
    Thickness
    Color
    "f(x)"=Power(Base=@n51, Exp=@n52)
   ) {materialFunction=/Engine/Functions/Engine_MaterialFunctions01/Debug/PlotFunctionOnGraph}."f(x) = y")
  NamedRerouteDeclaration(None=FunctionInput(Preview=StaticBool()) {bUsePreviewValueAsDefault=true, inputName=IsTree, inputType=FunctionInput_StaticBool}) {name=IsTreeBool, nodeColor=<rgba 1 0 1 1>}
  n53 = Add(
   A=Multiply(A=NamedRerouteUsage(), B=ScalarParameter() {defaultValue=2.75, group=SnowProgression, parameterName=TPS-HeightMask2Mod})
   B=ScalarParameter() {defaultValue=-2, group=SnowProgression, parameterName=TPS-HeightMask2Add}
  )
  NamedRerouteDeclaration(None=MaterialFunctionCall(Transition Phase=NamedRerouteUsage(), Height Texture=Saturate(None=@n53), Contrast=Constant() {r=0.25}) {materialFunction=&HeightLerp}.Alpha) {name=MinimapProgressionSetup, nodeColor=<rgba 1 0 1 1>}
  n54 = ComponentMask(None=WorldPosition() {worldPositionShaderOffset=WPT_ExcludeAllShaderOffsets}.XYZ) {g=true, r=true}
  n55 = Constant2Vector() {g=32047, r=-34689}
  n56 = StaticSwitch(
   True=MaterialFunctionCall!(A=Add(A=@n54, B=@n55), B, Bounds=Constant() {r=340000}, Edge Falloff=Constant() {r=0.1}) {materialFunction=/Engine/Functions/Engine_MaterialFunctions01/Gradient/BoxMask-2D}.Result
   False=Constant() {r=1}
   Value=FunctionInput(Preview=StaticBool() {value=true}) {bUsePreviewValueAsDefault=true, description="default=true", inputName=UseBR-IslandMask, inputType=FunctionInput_StaticBool}
  )
  NamedRerouteDeclaration(None=Multiply(A=NamedRerouteUsage(), B=@n56)) {name=Output-SimpleMask, nodeColor=<rgba 1 0 1 1>}
  n57 = FunctionInput(Preview=Constant() {r=1}) {
   bUsePreviewValueAsDefault=true
   desc=Texture mask or landscape layer
   description="For example, this could be a landscape layer."
   inputName=AlternateStaticSnowMask
   inputType=FunctionInput_Scalar
   sortPriority=-300
  }
  n58 = FunctionInput(Preview=StaticBool()) {
   bUsePreviewValueAsDefault=true
   description="Default = False"
   inputName=UseAlternateStaticSnowMask
   inputType=FunctionInput_StaticBool
   sortPriority=-400
  }
  n59 = StaticSwitch(True=@n57, False=NamedRerouteUsage(), Value=@n58)
  n60 = LinearInterpolate(
   A=ScalarParameter() {defaultValue=-0.5, group=SnowProgression, parameterName=RenderTarget-StaticSnowMask-LA}
   B=ScalarParameter() {defaultValue=1.25, group=SnowProgression, parameterName=RenderTarget-StaticSnowMask-LB}
   Alpha=@n59
  )
  n61 = NamedRerouteUsage()
  n62 = FunctionInput(Preview=StaticBool() {value=true}) {bUsePreviewValueAsDefault=true, description="default=true", inputName="UseRoad&&OcclusionMask", inputType=FunctionInput_StaticBool}
  n63 = Multiply(A=NamedRerouteUsage(), B=StaticSwitch(True=NamedRerouteUsage(), False=Constant() {r=1}, Value=@n62))
  n64 = FunctionInput(Preview=StaticBool()) {
   bUsePreviewValueAsDefault=true
   description="Default = False. This is for the year-round snow in the mountains."
   inputName=IsStaticSnow
   inputType=FunctionInput_StaticBool
   sortPriority=-500
  }
  NamedRerouteDeclaration(None=StaticSwitch(
    True=StaticSwitch(True=@n60, False=@n59, Value=@n61)
    False=Multiply(A=@n56, B=@n63)
    Value=NamedRerouteDeclaration(None=@n64) {name=IsStaticSnow, nodeColor=<rgba 1 0 1 1>}
   )) {name=OcclusionMask, nodeColor=<rgba 1 0 1 1>}
  n65 = TextureSample(
   UVs=MaterialFunctionCall!(Scale=Constant() {r=600}) {materialFunction=/Game/Athena/Apollo/Environments/Landscape/Materials/Biome/CommonMaterials/MF_SimpleWorldUV}.Result
   Tex=NamedRerouteUsage()
  ) {
   desc=T_Athena_Snow_01_MOBILE_M
   samplerSource=SSM_Wrap_WorldGroupSettings
   samplerType=SAMPLERTYPE_Normal
   texture=&FlatNormalBlue
  }
  NamedRerouteDeclaration(None=@n65.R) {name=SnowDetailMask-R, nodeColor=<rgba 1 0 1 1>}
  n66 = TextureCoordinate()
  n67 = ScalarParameter() {defaultValue=15, group=SnowProgression, parameterName=EdgeBreakupSizeG-Minimap}
  n68 = MaterialFunctionCall(Scale=Constant() {r=1200}) {materialFunction=&MF_SimpleWorldUV}
  n69 = TextureSample(
   UVs=StaticSwitch(
    True=Multiply(A=@n66, B=@n67)
    False=NamedRerouteDeclaration(None=@n68.Result) {name=NoiseScale1, nodeColor=<rgba 1 0 1 1>}
    Value=NamedRerouteUsage()
   )
   Tex=NamedRerouteUsage()
  ) {
   desc=T_Athena_Snow_01_MOBILE_M
   samplerSource=SSM_Wrap_WorldGroupSettings
   samplerType=SAMPLERTYPE_Normal
   texture=&FlatNormalBlue
  }
  NamedRerouteDeclaration(None=@n69.G) {name=SnowDetailMask-G, nodeColor=<rgba 1 0 1 1>}
  n70 = CurveAtlasRowParameter(CurveTime=CollectionParameter() {collection=&FortniteMaterialParameters, parameterName=TimeOfDay_0-1}) {
   atlas=&/LandscapeCurveAtlas
   curve=&/LandscapeCurve-TOD
   desc="0=morn 1=day 0=eve 1=night"
  }
  NamedRerouteDeclaration(None=Saturate(None=@n70.A)) {name=TOD-Lerp, nodeColor=<rgba 1 0 1 1>}
  n71 = SetMaterialAttributes!(MaterialAttributes)
  n72 = GetMaterialAttributes(None=NamedRerouteDeclaration(None=FunctionInput(Preview=@n71) {bUsePreviewValueAsDefault=true, inputName=Material, inputType=FunctionInput_MaterialAttributes, sortPriority=500}) {name=InputMaterial, nodeColor=<rgba 1 0 1 1>}) {attributeGetTypes [
    #2
    #5
    #3
    #4
   ]}
  NamedRerouteDeclaration(None=@n72.Normal) {name=InputMaterial-Normals, nodeColor=<rgba 1 0 1 1>}
  n73 = NamedRerouteUsage()
  n74 = NamedRerouteUsage()
  n75 = ComponentMask(None=VertexNormalWS()) {b=true}
  n76 = Multiply(A=@n75, B=@n75)
  n77 = NamedRerouteDeclaration(None=Saturate(None=Multiply(A=@n76, B=@n76))) {name=SlopePowerMask, nodeColor=<rgba 1 0 1 1>}
  n78 = NamedRerouteUsage()
  n79 = Add(
   A=Multiply(A=ScalarParameter() {defaultValue=0.5, group=SnowProgression, parameterName=IncreaseSnowUsingSlope-AddAmt}, B=OneMinus(None=@n77))
   B=@n78
  )
  n80 = FunctionInput(Preview=StaticBool() {value=true}) {bUsePreviewValueAsDefault=true, description="default=true", inputName=IncreaseSnowUsingSlope, inputType=FunctionInput_StaticBool}
  n81 = Multiply(
   A=StaticSwitch(True=Saturate(None=@n79), False=@n78, Value=NamedRerouteDeclaration(None=@n80) {name=IncreaseSnowUsingSlope, nodeColor=<rgba 1 0 1 1>})
   B=StaticSwitch(True=NamedRerouteUsage(), False=NamedRerouteUsage(), Value=NamedRerouteUsage())
  )
  n82 = Multiply(A=NamedRerouteUsage(), B=StaticSwitch(True=NamedRerouteUsage(), False=@n81, Value=NamedRerouteUsage()))
  n83 = NamedRerouteUsage()
  n84 = Constant() {r=1}
  n85 = FunctionInput(Preview=StaticBool()) {bUsePreviewValueAsDefault=true, description="Default = False", inputName=UseSmoothBorder, inputType=FunctionInput_StaticBool}
  n86 = Constant() {r=1}
  n87 = FunctionInput(Preview=StaticBool()) {
   bUsePreviewValueAsDefault=true
   description="Default = False"
   inputName=UseAdditionalMaskToRemoveSnow
   inputType=FunctionInput_StaticBool
   sortPriority=4
  }
  n88 = Multiply(
   A=StaticSwitch(
    True=Multiply(A=@n73, B=@n74)
    False=MaterialFunctionCall(Transition Phase=@n82, Height Texture=@n83, Contrast=@n84) {materialFunction=&HeightLerp}.Alpha
    Value=@n85
   )
   B=StaticSwitch(
    True=FunctionInput(Preview=@n86) {bUsePreviewValueAsDefault=true, inputName=AdditionalMaskToRemoveSnow, inputType=FunctionInput_Scalar}
    False=@n86
    Value=@n87
   )
  )
  NamedRerouteDeclaration(None=@n88) {name=FinalMaterialSnowMask, nodeColor=<rgba 1 0 1 1>}
  n89 = StaticBool()
  n90 = FunctionInput(Preview=StaticBool()) {bUsePreviewValueAsDefault=true, description="default=false", inputName=IsForRenderTarget, inputType=FunctionInput_StaticBool, sortPriority=3}
  n91 = MaterialFunctionCall!(
   SnowStageMask=NamedRerouteUsage()
   SnowAreaDetailMask=NamedRerouteUsage()
   UseMinimapCoords=NamedRerouteDeclaration(None=FunctionInput(Preview=@n89) {bUsePreviewValueAsDefault=true, description="default=false", inputName=UseMinimapCoords, inputType=FunctionInput_StaticBool}) {name=UseMinimapCoords, nodeColor=<rgba 1 0 1 1>}
   IsForRenderTarget=NamedRerouteDeclaration(None=@n90) {name=IsForRenderTarget, nodeColor=<rgba 1 0 1 1>}
  ) {materialFunction=&/MF_Snow_Progression_Mask}
  NamedRerouteDeclaration(None=@n91.SnowAreaDetail-G) {
   name=SnowAreaDetail-G
   desc="T_SnowMask_Progression\r\nG= Mask for where snow shouldn't be\r\n(water, under trees, POI's, etc)\r\nOR\r\nStatic snow mask"
   nodeColor=<rgba 1 0 1 1>
  }
  NamedRerouteDeclaration(None=@n72.Roughness) {name=InputMaterial-Roughness, nodeColor=<rgba 1 0 1 1>}
  NamedRerouteDeclaration(None=@n72.Specular) {name=InputMaterial-Spec, nodeColor=<rgba 1 0 1 1>}
  NamedRerouteDeclaration(None=@n72.BaseColor) {name=InputMaterial-Basecolor, nodeColor=<rgba 1 0 1 1>}
  NamedRerouteDeclaration(None=@n72.MaterialAttributes) {name=InputMaterial-Attributes, nodeColor=<rgba 1 0 1 1>}
  n92 = TextureObject() {
   desc="T_Athena_Snow_01_MOBILE_M\r\nUse this texture on mobile/switch or it will cause a blocky mess!\r\nThere was a reason we set it to a 512 uncompressed, but I can't remember why exactly.\r\nIt needs to be kept at this size for default now because the other one becomes pixellated\r\nat this larger scale.\r\nNote: Previously, the blue channel was being used as a general-area snow mask. It\r\nwas divided into 9 sections for 9 different snow amount levels."
   samplerType=SAMPLERTYPE_LinearColor
   texture=/Game/Athena/Environments/Landscape/Textures/T_Athena_Snow_01_MOBILE_M
  }
  NamedRerouteDeclaration(None=@n92) {name=SnowChannelPack, nodeColor=<rgba 1 0 1 1>}
  n93 = TextureObject() {
   desc="Set this to the new, static snow texture mask"
   samplerType=SAMPLERTYPE_LinearColor
   texture=/Game/Athena/Apollo/Environments/Landscape/Materials/Biome/CommonMaterials/T_SnowMask_Static
  }
  NamedRerouteDeclaration(None=StaticSwitch(
    True=@n93
    False=TextureObject() {desc=T_SnowMask_Progression, samplerType=SAMPLERTYPE_Normal, texture=&FlatNormalBlue}
    Value=NamedRerouteUsage()
   )) {name=SnowAreaDetail, nodeColor=<rgba 1 0 1 1>}
  n94 = FunctionInput(Preview=StaticBool() {value=true}) {bUsePreviewValueAsDefault=true, description="Default = True", inputName=UseSparkle, inputType=FunctionInput_StaticBool, sortPriority=4}
  NamedRerouteDeclaration(None=@n94) {name=UseSparkle, nodeColor=<rgba 1 0 1 1>}
  n95 = TextureCoordinate()
  n96 = ScalarParameter() {defaultValue=25, group=SnowProgression, parameterName=EdgeBreakupSizeG2-Minimap}
  n97 = MaterialFunctionCall(Scale=Constant() {r=3000}) {materialFunction=&MF_SimpleWorldUV}
  n98 = TextureSample(
   UVs=StaticSwitch(
    True=Multiply(A=@n95, B=@n96)
    False=NamedRerouteDeclaration(None=@n97.Result) {name=NoiseScale2, nodeColor=<rgba 1 0 1 1>}
    Value=NamedRerouteUsage()
   )
   Tex=NamedRerouteUsage()
  ) {
   desc=T_Athena_Snow_01_MOBILE_M
   samplerSource=SSM_Wrap_WorldGroupSettings
   samplerType=SAMPLERTYPE_Normal
   texture=&FlatNormalBlue
  }
  NamedRerouteDeclaration(None=@n98.G) {name=SnowDetailMask-G2, nodeColor=<rgba 1 0 1 1>}
  n99 = NamedRerouteUsage()
  n100 = FunctionInput(Preview=StaticBool()) {
   bUsePreviewValueAsDefault=true
   description="Default = False"
   inputName=IsTangentSpaceInput
   inputType=FunctionInput_StaticBool
   sortPriority=4
  }
  NamedRerouteDeclaration(None=StaticSwitch(True=Transform(None=@n99), False=@n99, Value=NamedRerouteDeclaration(None=@n100) {name=IsTangentSpaceInput, nodeColor=<rgba 1 0 1 1>})) {name=InputNormalsForMask, nodeColor=<rgba 1 0 1 1>}
  n101 = Add(A=NamedRerouteUsage()) {constB=-0.5}
  NamedRerouteDeclaration(None=Multiply(A=@n91.SnowAreaDetail-B, B=Saturate(None=@n101))) {name=TiltedMask, nodeColor=<rgba 1 0 1 1>}
  NamedRerouteDeclaration(None=LinearInterpolate(A=Constant3Vector() {constant=<rgba 0.285 0.4009 0.5 1>}, B=Constant3Vector() {constant=<rgba 0.2 0.3078 0.4 1>}, Alpha=NamedRerouteUsage())) {name=ShadowColor, nodeColor=<rgba 1 0 1 1>}
  n102 = LinearInterpolate(A=NamedRerouteUsage(), B=NamedRerouteUsage(), Alpha=Constant() {r=0.5})
  NamedRerouteDeclaration(None=LinearInterpolate(A=SquareRoot(None=@n102), B=Constant() {r=1}, Alpha=Constant() {r=0.25})) {name=AdditionalNoise, nodeColor=<rgba 1 0 1 1>}
  n103 = TextureCoordinate()
  n104 = ScalarParameter() {defaultValue=200, group=SnowProgression, parameterName=EdgeBreakupSize-Minimap}
  n105 = Constant() {desc=OG, r=8000}
  n106 = TextureSample(
   UVs=StaticSwitch(
    True=Multiply(A=@n103, B=@n104)
    False=MaterialFunctionCall(Scale=@n105) {materialFunction=&MF_SimpleWorldUV}.Result
    Value=NamedRerouteUsage()
   )
   Tex=NamedRerouteUsage()
  ) {
   desc=T_Athena_Snow_01_MOBILE_M
   samplerSource=SSM_Wrap_WorldGroupSettings
   samplerType=SAMPLERTYPE_Normal
   texture=&FlatNormalBlue
  }
  NamedRerouteDeclaration(None=@n106.A) {name=MacroNoise1, nodeColor=<rgba 1 0 1 1>}
  n107 = TextureSample(
   UVs=MaterialFunctionCall(Scale=Constant() {r=1900}) {materialFunction=&MF_SimpleWorldUV}.Result
   Tex=NamedRerouteUsage()
  ) {
   desc=Macro1
   samplerSource=SSM_Wrap_WorldGroupSettings
   samplerType=SAMPLERTYPE_Normal
   texture=&FlatNormalBlue
  }
  n108 = TextureSample(
   UVs=MaterialFunctionCall(Scale=Constant() {r=3000}) {materialFunction=&MF_SimpleWorldUV}.Result
   Tex=NamedRerouteUsage()
  ) {
   desc=Macro2
   samplerSource=SSM_Wrap_WorldGroupSettings
   samplerType=SAMPLERTYPE_Normal
   texture=&FlatNormalBlue
  }
  n109 = NamedRerouteUsage()
  n110 = NamedRerouteUsage()
  n111 = TextureSample(UVs=NamedRerouteUsage(), Tex=@n110) {samplerSource=SSM_Wrap_WorldGroupSettings, samplerType=SAMPLERTYPE_Normal, texture=&FlatNormalBlue}
  n112 = TextureSample(UVs=NamedRerouteUsage(), Tex=@n110) {samplerSource=SSM_Wrap_WorldGroupSettings, samplerType=SAMPLERTYPE_Normal, texture=&FlatNormalBlue}
  n113 = LinearInterpolate(A=@n111.RGB, B=@n112.RGB)
  n114 = Constant() {r=3}
  n115 = Add(A=Multiply(A=NamedRerouteUsage(), B=Constant() {r=3.25}), B=Constant() {r=-0.3})
  n116 = MaterialFunctionCall(
   BaseNormal=Multiply(A=LinearInterpolate(A=@n107.RGB, B=@n108.RGB, Alpha=@n109), B=Constant3Vector() {constant=<rgba 0.6 0.6 1 0>})
   AdditionalNormal=LinearInterpolate(A=Multiply(A=@n113, B=@n114), B=Constant3Vector() {constant=<rgba 0 0 1 0>}, Alpha=Saturate(None=@n115))
  ) {materialFunction=&BlendAngleCorrectedNormals}
  n117 = Multiply(A=Constant3Vector() {constant=<rgba 0.25 0.25 1 1>}, B=@n116.Result)
  n118 = SphereMask(A=CameraPositionWS(), B=WorldPosition().XYZ, Radius=Constant() {r=60000}, Hardness=Constant())
  n119 = FunctionInput(Preview=StaticBool()) {
   bUsePreviewValueAsDefault=true
   description="Default = False"
   inputName=UseVertexNormalsForBasicSnow
   inputType=FunctionInput_StaticBool
   sortPriority=4
  }
  n120 = StaticSwitch(
   True=VertexNormalWS()
   False=MaterialFunctionCall!(Normals=LinearInterpolate(A=@n117, B=@n116.Result, Alpha=@n118), TransformToTangentSpace, Normal Strength=Constant() {r=1.5}) {materialFunction=/Game/Athena/Apollo/Environments/Landscape/Materials/Biome/CommonMaterials/MF_HQ_WorldAlignedNormal}.Result
   Value=@n119
  )
  n121 = FunctionInput(Preview=StaticBool() {value=true}) {
   bUsePreviewValueAsDefault=true
   description="Default = True"
   inputName=UseTangentSnowNormals
   inputType=FunctionInput_StaticBool
   sortPriority=4
  }
  NamedRerouteDeclaration(None=StaticSwitch(True=Transform(None=@n120) {transformSourceType=TRANSFORMSOURCE_World, transformType=TRANSFORM_Tangent}, False=@n120, Value=@n121)) {name=SnowNormals, nodeColor=<rgba 1 0 1 1>}
  n122 = TextureCoordinate()
  n123 = ScalarParameter() {defaultValue=100, group=SnowProgression, parameterName=EdgeBreakupSize2-Minimap}
  n124 = Constant() {r=22000}
  n125 = TextureSample(
   UVs=StaticSwitch(
    True=Multiply(A=@n122, B=@n123)
    False=MaterialFunctionCall(Scale=@n124) {materialFunction=&MF_SimpleWorldUV}.Result
    Value=NamedRerouteUsage()
   )
   Tex=NamedRerouteUsage()
  ) {
   desc=T_Athena_Snow_01_MOBILE_M
   samplerSource=SSM_Wrap_WorldGroupSettings
   samplerType=SAMPLERTYPE_Normal
   texture=&FlatNormalBlue
  }
  NamedRerouteDeclaration(None=@n125.A) {name=MacroNoise2, nodeColor=<rgba 1 0 1 1>}
  n126 = TextureObject() {
   desc=T_Athena_Snow_01_MOBILE_N
   samplerType=SAMPLERTYPE_Normal
   texture=/Game/Athena/Environments/Landscape/Textures/T_Athena_Snow_01_MOBILE_N
  }
  NamedRerouteDeclaration(None=@n126) {name=MicroNormalTextureObj, nodeColor=<rgba 1 0 1 1>}
  n127 = TextureObject() {desc=T_Artemis_Snow_N_01, samplerType=SAMPLERTYPE_Normal, texture=/Game/Environments/Artemis/Sets/Terrain/Textures/T_Artemis_Snow_N_01}
  NamedRerouteDeclaration(None=@n127) {name=Macro1NormalTextureObj, nodeColor=<rgba 1 0 1 1>}
  n128 = TextureObject() {desc=T_Artemis_Snow_N_03, samplerType=SAMPLERTYPE_Normal, texture=/Game/Environments/Artemis/Sets/Terrain/Textures/T_Artemis_Snow_N_03}
  NamedRerouteDeclaration(None=@n128) {name=Macro2NormalTextureObj, nodeColor=<rgba 1 0 1 1>}
  n129 = LinearInterpolate(A=NamedRerouteUsage(), B=VertexNormalWS(), Alpha=ScalarParameter() {group=SnowProgression, parameterName=Noise-SlopeMask-UseVertexNormals})
  n130 = Multiply(
   A=Saturate(None=ComponentMask(None=@n129) {b=true})
   B=ScalarParameter() {defaultValue=3, group=SnowProgression, parameterName=SnowHeightMask-Mod}
  )
  n131 = ScalarParameter() {defaultValue=-0.15, group=SnowProgression, parameterName=SnowHeightMask-Add}
  NamedRerouteDeclaration(None=Saturate(None=Add(A=@n130, B=@n131))) {name=SlopeMask, nodeColor=<rgba 1 0 1 1>}
  NamedRerouteDeclaration(None=@n82) {name=NormalNoiseMask, nodeColor=<rgba 1 0 1 1>}
  n132 = LinearInterpolate(A=Constant() {r=-0.15}, B=Constant() {r=-1}, Alpha=NamedRerouteUsage())
  n133 = LinearInterpolate(A=Constant() {r=5}, B=Constant() {r=2}, Alpha=NamedRerouteUsage())
  n134 = LinearInterpolate(A=PixelNormalWS(), B=VertexNormalWS(), Alpha=ScalarParameter() {group=SnowProgression, parameterName=FSS-BlendNormalType})
  n135 = SkyAtmosphereLightDirection()
  n136 = Power(Base=Saturate(None=DotProduct(A=@n134, B=@n135)), Exp=Constant() {r=3.5})
  NamedRerouteDeclaration(None=Saturate(None=LinearInterpolate(A=@n132, B=@n133, Alpha=@n136))) {name=ShadowMask, nodeColor=<rgba 1 0 1 1>}
  n137 = Constant() {r=800}
  n138 = Divide(A=CameraPositionWS(), B=Constant() {r=10000})
  n139 = TextureSample(
   UVs=Add(
    A=MaterialFunctionCall(Scale=@n137) {materialFunction=&MF_SimpleWorldUV}.Result
    B=ComponentMask(None=@n138) {g=true, r=true}
   )
   Tex=NamedRerouteUsage()
  ) {
   desc=T_Athena_Snow_01_MOBILE_M
   samplerSource=SSM_Wrap_WorldGroupSettings
   samplerType=SAMPLERTYPE_Normal
   texture=&FlatNormalBlue
  }
  n140 = Multiply(A=@n139.R, B=@n139.R)
  n141 = Multiply(A=@n139.R, B=Multiply(A=@n140, B=@n140))
  n142 = Constant() {r=1.5}
  NamedRerouteDeclaration(None=Saturate(None=Multiply(A=@n141, B=@n142))) {name=Sparkle-Roughness, nodeColor=<rgba 1 0 1 1>}
  n143 = NamedRerouteUsage()
  n144 = OneMinus(None=NamedRerouteUsage())
  n145 = Constant() {r=0.5}
  NamedRerouteDeclaration(None=Power(Base=LinearInterpolate(A=@n143, B=@n144, Alpha=@n145), Exp=Constant() {r=1.2})) {name=MacroNoiseCombined, nodeColor=<rgba 1 0 1 1>}
  NamedRerouteDeclaration(None=Constant3Vector() {constant=<rgba 0.26 0.3355 0.4 1>}) {name=OutputSnowAlbedoForRVT, nodeColor=<rgba 1 0 1 1>}
  NamedRerouteDeclaration(None=Power(Base=@n88, Exp=Constant() {r=1.5})) {name=SnowMaskForRVT, nodeColor=<rgba 1 0 1 1>}
  n146 = Constant()
  n147 = Constant() {r=2}
  n148 = Constant() {r=4}
  n149 = NamedRerouteUsage()
  n150 = Subtract(A=OneMinus(None=NamedRerouteUsage()), B=Multiply(A=@n139.R, B=Constant() {r=0.55}) {constB=0.75})
  n151 = Multiply(A=Multiply(A=LinearInterpolate(A=@n147, B=@n148, Alpha=@n149), B=Saturate(None=@n150)), B=NamedRerouteUsage())
  n152 = Constant()
  NamedRerouteDeclaration(None=ShadingPathSwitch(Default=QualitySwitch(Default=@n146, Low=@n146, High=@n151, Medium=@n146, Epic=@n151), Forward=@n152, Mobile=@n152)) {name=Sparkle-Emissive, nodeColor=<rgba 1 0 1 1>}
  NamedRerouteDeclaration(None=NamedRerouteUsage()) {name=SnowCoverage-OR-FallingSnowCoverage-Scalar, nodeColor=<rgba 1 0 1 1>}
  NamedRerouteDeclaration(None=Multiply(A=NamedRerouteUsage(), B=@n91.Snow-SUB-UV-Mask)) {name=ProgressionCoverageMain, nodeColor=<rgba 1 0 1 1>}
  Power(Base=@n139.R, Exp=Constant() {r=5})
  Power(Base=@n75, Exp=Constant() {r=4})
  n153 = NamedRerouteDeclaration(None=@n32) {name=SnowCoverageScalar, nodeColor=<rgba 1 0 1 1>}
  n154 = FunctionInput(Preview=StaticBool()) {bUsePreviewValueAsDefault=true, inputName=IsFallingSnowFX, inputType=FunctionInput_StaticBool}
  Reroute(None=Saturate(None=StaticSwitch(True=@n33, False=@n153, Value=@n154)))
  Reroute(None=Max(A=NamedRerouteUsage() {desc=Tilted mask}, B=@n81))
  n155 = Add(A=@n91.SnowAreaDetail-G, B=ScalarParameter() {group=SnowProgression, parameterName=DEBUG-Override-SnowOcclusionTextureMask})
  Reroute(None=LinearInterpolate(A=Saturate(None=@n155), B=@n24, Alpha=ScalarParameter() {group=SnowProgression, parameterName=DEBUG-Override-UseBiomeD-OcclusionTextureMask}))
  ScalarParameter() {defaultValue=0.25, group=SnowProgression, parameterName=TPS-HeightBlendContrast}
  ScalarParameter() {defaultValue=3.25, group=SnowProgression, parameterName=Snow-NormalMaskSmallModA}
  ScalarParameter() {defaultValue=-0.3, group=SnowProgression, parameterName=Snow-NormalMaskSmallAddF}
  ScalarParameter() {defaultValue=1.5, group=SnowProgression, parameterName=SPK-modroughx}
  ScalarParameter() {defaultValue=4, group=SnowProgression, parameterName=SPK-mod-day}
  ScalarParameter() {defaultValue=0.7, group=SnowProgression, parameterName=SnowAlbedo}
  ScalarParameter() {defaultValue=600, group=SnowProgression, parameterName=SnowRoughScale}
  ScalarParameter() {defaultValue=0.5, group=SnowProgression, parameterName=SnowFlatSpec}
  ScalarParameter() {defaultValue=3000, desc="Shared!", group=SnowProgression, parameterName=SnowNoiseScale2}
  ScalarParameter() {defaultValue=0.5, group=SnowProgression, parameterName=SnowRoughMobile}
  ScalarParameter() {group=SnowProgression, parameterName=DEBUG-Override-SnowProgressionMasks}
  ScalarParameter() {defaultValue=1200, group=SnowProgression, parameterName=SnowNoiseScale}
  ScalarParameter() {defaultValue=4, group=SnowProgression, parameterName=IncreaseSnowUsingSlope-Power}
  ScalarParameter() {defaultValue=0.25, group=SnowProgression, parameterName=DGDG-AddNoise-Soft}
  ScalarParameter() {defaultValue=3.5, group=SnowProgression, parameterName=FSS-BlendPow}
  ScalarParameter() {defaultValue=-0.15, group=SnowProgression, parameterName=FSS-A-Morning}
  ScalarParameter() {defaultValue=1, group=SnowProgression, parameterName=DGDG-OcclusionMaskthing1-Cont}
  ScalarParameter() {defaultValue=8000, group=SnowProgression, parameterName=EdgeBreakup1-masktile}
  ScalarParameter() {defaultValue=5, group=SnowProgression, parameterName=SPK-FresPow2}
  ScalarParameter() {defaultValue=0.35, group=SnowProgression, parameterName=SnowAlbedo-Mobile}
  ScalarParameter() {defaultValue=2, group=SnowProgression, parameterName=SPK-mod-morn}
  ScalarParameter() {defaultValue=800, group=SnowProgression, parameterName=SnowSparkleMaskScale}
  ScalarParameter() {defaultValue=1900, group=SnowProgression, parameterName=Snow-NormalScale-TEST1}
  ScalarParameter() {defaultValue=22000, group=SnowProgression, parameterName=EdgeBreakup2-masktile}
  ScalarParameter() {defaultValue=3000, group=SnowProgression, parameterName=Snow-NormalScale-TEST2}
  ScalarParameter() {defaultValue=60000, group=SnowProgression, parameterName=Snow-NormalDistanceRad}
  ScalarParameter() {defaultValue=-1, group=SnowProgression, parameterName=FSS-A-Day}
  ScalarParameter() {defaultValue=5, group=SnowProgression, parameterName=FSS-B-Morning}
  ScalarParameter() {defaultValue=2, group=SnowProgression, parameterName=FSS-B-Day}
  ScalarParameter() {defaultValue=1.2, group=SnowProgression, parameterName=NoiseSnowHeightMask-PowtestK}
  ScalarParameter() {defaultValue=10000, group=SnowProgression, parameterName=SPK-Div111}
  ScalarParameter() {defaultValue=3, group=SnowProgression, parameterName=Snow-SmallNormalsMod1}
  n156 = Subtract(A=ComponentMask(None=WorldPosition().XYZ) {b=true}, B=Constant() {r=-3570})
  n157 = Constant() {r=-500}
  n158 = OneMinus(None=Saturate(None=Divide(A=@n156, B=@n157)))
  StaticSwitch(
   True=If(
    A=NamedRerouteUsage()
    B=ScalarParameter() {defaultValue=0.94, group=SnowProgression, parameterName=Lobby-Trigger}
    "A > B"=@n158
    "A == B"=@n158
    "A < B"=Constant()
   )
   False=@n63
   Value=FunctionInput(Preview=StaticBool()) {bUsePreviewValueAsDefault=true, description="default=false", inputName=IsLobbyIsland, inputType=FunctionInput_StaticBool}
  )
  StaticSwitch(
   True=If(
    A=ScalarParameter() {defaultValue=0.1, group=SnowProgression, parameterName=Snow-IF-B}
    B=NamedRerouteUsage()
    "A > B"=Constant()
    "A == B"=@n91.Snow-SUB-UV-Mask
    "A < B"=@n91.Snow-SUB-UV-Mask
   )
   False=@n91.Snow-SUB-UV-Mask
   Value=NamedRerouteUsage()
  )
  n159 = Constant() {r=1}
  StaticSwitch(True=Saturate(None=Multiply(A=@n159, B=@n88)), False=@n88, Value=NamedRerouteUsage())
  TextureSample() {
   desc=FlatNormalBlue
   samplerSource=SSM_Wrap_WorldGroupSettings
   samplerType=SAMPLERTYPE_Normal
   texture=&FlatNormalBlue
  }
  VectorParameter() {group=SnowProgression, parameterName=SnowShadowColor-Day, defaultValue=<rgba 0.2 0.3078 0.4 1>}
  VectorParameter() {group=SnowProgression, parameterName=SnowShadowColor-Morning, defaultValue=<rgba 0.285 0.4009 0.5 1>}
  VectorParameter() {parameterName=Snow-DistanceNormalStrength, defaultValue=<rgba 0.25 0.25 1 1>}
 ]}
