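# Apollo landscape material function that pushes geometry, apparently imposters, out of view when the supplied position sits below the water level.
# It adds WaterLevelBias to WaterLevel from the Apollo terrain parameter collection and compares the sum with ImposterPosition's Z.
# Below that height, World Position Offset becomes the imposter position with Z dropped by a huge constant; above, it passes through.
# Outputs are modified material attributes, the biased water level, and a standalone WPO using WPO_Input.
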MaterialFunction /Game/Athena/Apollo/Environments/Landscape/Materials/Biome/UniqueBiomeMaterialFunctions/MF_Apollo_WaterLevel
graph {nodes [
  n1 = GetMaterialAttributes(None=FunctionInput() {inputType=FunctionInput_MaterialAttributes}) {attributeGetTypes [#1, #2]}
  n2 = FunctionInput() {inputName=ImposterPosition}
  n3 = ComponentMask(None=@n2) {b=true}
  n4 = Add(
   A=CollectionParameter() {collection=/Game/Athena/Apollo/Environments/Landscape/MPC/Apollo_MPC_Terrain, parameterName=WaterLevel}
   B=ScalarParameter() {defaultValue=-20, parameterName=WaterLevelBias}
  )
  n5 = AppendVector(A=ComponentMask(None=@n2) {g=true, r=true}, B=Subtract(A=@n3, B=Constant() {r=80000}))
  FunctionOutput(None=SetMaterialAttributes!(
    MaterialAttributes=@n1.MaterialAttributes
    WorldPositionOffset=If(A=@n3, B=@n4, "A > B"=@n1.WorldPositionOffset, "A < B"=@n5)
    BaseColor=@n1.MaterialAttributes
   ) {attributeSetTypes [#1, #2]})
  FunctionOutput(None=@n4) {description=Waterlevel with Bias, outputName=WaterLevel}
  FunctionOutput(None=If(A=@n3, B=@n4, "A > B"=FunctionInput() {bUsePreviewValueAsDefault=true, inputName=WPO_Input}, "A < B"=@n5)) {outputName=WPO}
  WorldPosition()
 ]}
