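# Landscape biome material function that builds the autumn leaves layer from tiling forest leaf color, mask and normal textures.
# Color is desaturated and tinted by constant division, masks drive specular and roughness, and the flattened normal blends with the incoming material's normal.
# Height multiplies leaf alpha by remapped world-space cloud noise and a boost.
# It also outputs a blend contrast and the landscape layer weight masked by leaf alpha.
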MaterialFunction /Game/Athena/Apollo/Environments/Landscape/Materials/Biome/UniqueBiomeMaterialFunctions/MF_Biome_Autumn_Leaves
graph {nodes [
  Constant() {r=1.4}
  Constant() {r=10}
  Constant() {r=-2}
  Constant() {r=10}
  Constant() {r=-2.5}
  Constant() {r=1300}
  Constant() {r=1.6}
  Constant() {r=2}
  n1 = Multiply(A=LandscapeLayerCoords() {mappingScale=10}, B=ScalarParameter() {defaultValue=1.6, group=Forest, parameterName=AutumnLeavesScale})
  n2 = TextureSample(UVs=@n1) {samplerSource=SSM_Wrap_WorldGroupSettings, texture=/Game/Environments/Apollo/Sets/Terrain/Textures/T_Apollo_Forest_Leaves_D_01}
  n3 = Desaturation(None=@n2.RGB, Fraction=Constant() {r=0.3})
  n4 = Constant3Vector() {constant=<rgba 1.8 1.9347 3 1>}
  n5 = TextureSample(UVs=@n1) {
   samplerSource=SSM_Wrap_WorldGroupSettings
   samplerType=SAMPLERTYPE_Masks
   texture=/Game/Environments/Apollo/Sets/Terrain/Textures/T_Apollo_Forest_Leaves_S_01
  }
  n6 = ScalarParameter() {defaultValue=0.5, group=Forest, parameterName=AutumnLeavesRough}
  n7 = TextureSample(UVs=@n1) {
   samplerSource=SSM_Wrap_WorldGroupSettings
   samplerType=SAMPLERTYPE_Normal
   texture=/Game/Environments/Apollo/Sets/Terrain/Textures/T_Apollo_Forest_Leaves_N_01
  }
  n8 = ScalarParameter() {defaultValue=0.5, group=Forest, parameterName=Autumn-NormalFlat}
  n9 = FunctionInput() {inputName=Material, inputType=FunctionInput_MaterialAttributes}
  n10 = SetMaterialAttributes!(
   MaterialAttributes
   BaseColor=Saturate(None=Divide(A=@n3, B=@n4))
   Specular=Multiply(A=ScalarParameter() {defaultValue=0.5, group=Forest, parameterName=AutumnLeavesSpec}, B=@n5.R)
   Roughness=Saturate(None=Add(A=@n5.B, B=@n6))
   Normal=LinearInterpolate(
    A=MaterialFunctionCall(Normal=@n7.RGB, Flatness=@n8) {materialFunction=&FlattenNormal}.Result
    B=GetMaterialAttributes(None=@n9) {attributeGetTypes [#1]}.Normal
    Alpha=ScalarParameter() {defaultValue=0.5, group=Forest, parameterName=Autumn-NormalUseIncomingNormalAmt}
   )
  ) {attributeSetTypes [
    #2
    #3
    #4
    #1
   ]}
  FunctionOutput(None=@n10) {outputName=Material}
  n11 = Divide(
   A=ComponentMask(None=WorldPosition() {worldPositionShaderOffset=WPT_ExcludeAllShaderOffsets}.XYZ) {g=true, r=true}
   B=ScalarParameter() {defaultValue=1300, parameterName=Leaves-Tex2Scale}
  )
  n12 = Saturate(None=LinearInterpolate(
    A=ScalarParameter() {defaultValue=-2, parameterName=LeavesTest-3}
    B=ScalarParameter() {defaultValue=10, parameterName=LeavesTest-2}
    Alpha=TextureSample(UVs=@n11) {samplerSource=SSM_Wrap_WorldGroupSettings, texture=&AtmosphericCloudNoise01}.R
   ))
  FunctionOutput(None=Multiply(A=ScalarParameter() {defaultValue=1.25, parameterName=Autumn-BoostHeight}, B=Multiply(A=@n12, B=@n2.A))) {outputName=Height, sortPriority=1}
  FunctionOutput(None=ScalarParameter() {defaultValue=3, parameterName=Autumn-BlendContrast}) {outputName=Contrast, sortPriority=2}
  n13 = Constant()
  FunctionOutput(None=Multiply(A=@n2.A, B=FunctionInput(Preview=@n13) {bUsePreviewValueAsDefault=true, inputName=LandscapeLayer, inputType=FunctionInput_Scalar})) {outputName=LandscapeLayer, sortPriority=1}
  ScalarParameter() {defaultValue=0.3, group=Forest, parameterName=Autumn-Desat}
  VectorParameter() {parameterName=Autumn-ColorDiv-Low, defaultValue=<rgba 1.8 1.9347 3 1>}
 ]}
