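# Material function that defines the forest floor layer for the Apollo landscape biome materials.
# Landscape layer coordinates, multiplied by a Forest Scale parameter, drive diffuse, mask and normal texture parameters using shared world-group wrap samplers.
# These set base color and normal, with the mask's red channel feeding both specular and roughness.
# It outputs the material attributes plus a Height value taken from the diffuse alpha.
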MaterialFunction /Game/Athena/Apollo/Environments/Landscape/Materials/Biome/UniqueBiomeMaterialFunctions/MF_Biome_Forest
graph {nodes [
  Constant() {r=1}
  n1 = Multiply(A=LandscapeLayerCoords() {mappingScale=10}, B=ScalarParameter() {defaultValue=1, group=Forest, parameterName=Forest Scale})
  n2 = TextureSampleParameter2D(UVs=@n1) {
   group=Forest
   parameterName=Forest_D
   samplerSource=SSM_Wrap_WorldGroupSettings
   texture=/Game/Environments/Apollo/Sets/Terrain/Textures/T_Apollo_Forest_Floor_D_01
  }
  n3 = TextureSampleParameter2D(UVs=@n1) {
   group=Forest
   parameterName=Forest_S
   samplerSource=SSM_Wrap_WorldGroupSettings
   samplerType=SAMPLERTYPE_Masks
   texture=/Game/Environments/Apollo/Sets/Terrain/Textures/T_Apollo_Forest_Floor_S_01
  }
  n4 = TextureSampleParameter2D(UVs=@n1) {
   group=Forest
   parameterName=Forest_N
   samplerSource=SSM_Wrap_WorldGroupSettings
   samplerType=SAMPLERTYPE_Normal
   texture=/Game/Environments/Apollo/Sets/Terrain/Textures/T_Apollo_Forest_Floor_N_01
  }
  n5 = SetMaterialAttributes!(MaterialAttributes, BaseColor=@n2.RGB, Specular=@n3.R, Roughness=@n3.R, Normal=@n4.RGB) {attributeSetTypes [
    #1
    #2
    #3
    #4
   ]}
  FunctionOutput(None=@n5) {outputName=Material}
  FunctionOutput(None=@n2.A) {outputName=Height, sortPriority=1}
 ]}
