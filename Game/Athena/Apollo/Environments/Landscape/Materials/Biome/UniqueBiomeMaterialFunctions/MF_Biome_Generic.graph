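# Generic landscape biome material function for Apollo terrain, turning a diffuse, normal and mask texture set into material attributes.
# Landscape layer coordinates, multiplied by an optional Scale input, drive all three samples through shared world-group wrap samplers.
# Diffuse RGB feeds base color, the normal map feeds Normal, and the mask's red channel feeds both Specular and Roughness.
# Outputs are the Material attributes plus a Height value from diffuse alpha.
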MaterialFunction /Game/Athena/Apollo/Environments/Landscape/Materials/Biome/UniqueBiomeMaterialFunctions/MF_Biome_Generic
graph {nodes [
  n1 = Multiply(
   A=LandscapeLayerCoords() {mappingScale=10}
   B=FunctionInput(Preview=Constant() {r=1}) {bUsePreviewValueAsDefault=true, inputName=Scale, inputType=FunctionInput_Scalar}
  )
  n2 = TextureSample(
   UVs=@n1
   Tex=FunctionInput(Preview=TextureObject() {texture=/Engine/EngineMaterials/DefaultDiffuse}) {inputName=Diffuse, inputType=FunctionInput_Texture2D, sortPriority=1}
  ) {samplerSource=SSM_Wrap_WorldGroupSettings}
  n3 = TextureSample(
   UVs=@n1
   Tex=FunctionInput(Preview=TextureObject() {samplerType=SAMPLERTYPE_Masks, texture=&T_Fortnite_Default_S}) {inputName=S_Mask, inputType=FunctionInput_Texture2D, sortPriority=3}
  ) {samplerSource=SSM_Wrap_WorldGroupSettings, samplerType=SAMPLERTYPE_Masks}
  n4 = TextureObject() {samplerType=SAMPLERTYPE_Normal, texture=&FlatNormalBlue}
  n5 = SetMaterialAttributes!(
   MaterialAttributes
   BaseColor=@n2.RGB
   Specular=@n3.R
   Roughness=@n3.R
   Normal=TextureSample(UVs=@n1, Tex=FunctionInput(Preview=@n4) {inputName=Normal, inputType=FunctionInput_Texture2D, sortPriority=2}) {samplerSource=SSM_Wrap_WorldGroupSettings, samplerType=SAMPLERTYPE_Normal}.RGB
  ) {attributeSetTypes [
    #1
    #2
    #3
    #4
   ]}
  FunctionOutput(None=@n5) {outputName=Material}
  FunctionOutput(None=@n2.A) {outputName=Height, sortPriority=1}
 ]}
