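# Material function for the Apollo landscape's heather biome, recoloring an input grass material with heather tints.
# GrassNoiseMask blends Base Color from Heather-ColorA toward a mix of ColorB and ColorC.
# With virtual texturing, that B-to-C mix follows the scaled inverse output of the runtime virtual texture level for distance variation; otherwise it uses a constant.
# The noise mask also passes through as Height, leaving the GrassHeightmap input unused.
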MaterialFunction /Game/Athena/Apollo/Environments/Landscape/Materials/Biome/UniqueBiomeMaterialFunctions/MF_Biome_Heather
graph {nodes [
  Constant3Vector() {constant=<rgba 0.0529 0.135 0.0203 1>}
  Constant3Vector() {constant=<rgba 0.418 0.0421 0.5 1>}
  Constant3Vector() {constant=<rgba 0.135 0.0898 0.079 1>}
  FunctionInput() {inputName=GrassHeightmap, inputType=FunctionInput_Scalar}
  n1 = VectorParameter() {parameterName=Heather-ColorA, defaultValue=<rgba 0.0313 0.08 0.012 1>}
  n2 = Multiply(
   A=ViewProperty() {property=MEVP_RuntimeVirtualTextureOutputLevel}.InvProperty
   B=ScalarParameter() {defaultValue=1.5, parameterName=RVT_DistNearHeather}
  )
  n3 = LinearInterpolate(
   A=VectorParameter() {parameterName=Heather-ColorB, defaultValue=<rgba 0.418 0.0421 0.5 1>}.RGB
   B=VectorParameter() {parameterName=Heather-ColorC, defaultValue=<rgba 0.1146 0.135 0.0425 1>}.RGB
   Alpha=VirtualTextureFeatureSwitch!(No=Constant(), Yes=Saturate(None=@n2))
  )
  n4 = FunctionInput() {inputName=GrassNoiseMask, inputType=FunctionInput_Scalar}
  FunctionOutput(None=SetMaterialAttributes!(
    MaterialAttributes=FunctionInput() {inputName=GrassMaterial, inputType=FunctionInput_MaterialAttributes}
    BaseColor=LinearInterpolate(A=@n1.RGB, B=@n3, Alpha=@n4)
   ) {attributeSetTypes [#1]}) {outputName=Material}
  FunctionOutput(None=@n4) {outputName=Height, sortPriority=1}
 ]}
