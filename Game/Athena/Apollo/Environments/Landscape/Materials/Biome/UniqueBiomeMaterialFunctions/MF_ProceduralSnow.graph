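# Material function for the Apollo landscape biome that produces procedural snow as standard, mobile and cheap material variants, plus material and grass masks.
# Snow coverage comes from a manual input or a fixed world-space sphere mask times an altitude ramp.
# The material mask height-lerps coverage against slope-weighted world-XY noise and the inverted heightmap, then fades on steep normals.
# The standard material blends world-aligned snow normals with vertex normals using that noise texture.
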MaterialFunction /Game/Athena/Apollo/Environments/Landscape/Materials/Biome/UniqueBiomeMaterialFunctions/MF_ProceduralSnow
graph {nodes [
  ComponentMask(None=VectorParameter() {parameterName=SnowMaskSM-B, defaultValue=<rgba 0 0 0 1>}.RGB) {g=true, r=true}
  n1 = WorldPosition()
  ComponentMask(None=@n1.XYZ) {g=true, r=true}
  n2 = ScalarParameter() {defaultValue=0.6, group=ProceduralSnow, parameterName=SnowAlbedo-Bright}
  n3 = VertexNormalWS()
  n4 = MaterialFunctionCall!(
   Diffuse Texture Object
   Spec Texture Object
   TextureSize=Constant() {r=4200}
   --------------------------
   Use seperate settings for Z projection
   Use seperate texture for Z projection
   Texture Size Z projection
   Z Plane Diffuse Texture Object
   Z Plane Spec Texture Object
   "--------------------------      "
   ProjectionTransitionContrast
   Use Texture Alpha Heightmap to Refine the Transistion
   HeightLerpTransitionContrast
   PickTheDiffuseHeightChannel
   TheDiffuseHeightChannel
   "--------------------------       "
   NormalTextureObject=TextureObject() {samplerType=SAMPLERTYPE_Normal, texture=/Game/Global/Textures/Default/Tileable_Patterns/T_Snow01_N}
   World Space Normals=StaticBool() {value=true}
   "--------------------------  "
   Export Float 4 Diffuse
   Export Float 4 Spec
   "--------------------------     "
   WorldPosition=WorldPosition() {worldPositionShaderOffset=WPT_ExcludeAllShaderOffsets}.XYZ
   World Space Normal
  ) {materialFunction=/Engine/Functions/Engine_MaterialFunctions02/Texturing/WorldAlignedTextures_Complex}
  n5 = ComponentMask(None=WorldPosition() {worldPositionShaderOffset=WPT_ExcludeAllShaderOffsets}.XYZ) {g=true, r=true}
  n6 = TextureSample(UVs=Multiply(A=@n5, B=Constant() {r=0.0002})) {
   samplerSource=SSM_Wrap_WorldGroupSettings
   samplerType=SAMPLERTYPE_LinearColor
   texture=/Game/Athena/Environments/Landscape/Textures/T_Athena_Snow_01_MOBILE_M
  }
  n7 = Multiply(A=@n6.A, B=Constant() {r=0.5})
  n8 = TextureSample(UVs=Multiply(A=@n5, B=Constant() {r=0.0018})) {
   samplerSource=SSM_Wrap_WorldGroupSettings
   samplerType=SAMPLERTYPE_LinearColor
   texture=&T_Athena_Snow_01_MOBILE_M
  }
  FunctionOutput(None=SetMaterialAttributes!(MaterialAttributes, BaseColor=@n2, Roughness=@n8.R, Normal=LinearInterpolate(A=@n3, B=@n4.Normal Z Texture, Alpha=@n7)) {attributeSetTypes [#1, #2, #3]}) {outputName=Material}
  n9 = Saturate(None=ComponentMask(None=FunctionInput() {bUsePreviewValueAsDefault=true, inputName=WorldnormalIn, previewValue=<0 0 1 1>}) {b=true})
  n10 = Multiply(A=Power(Base=@n9, Exp=Constant() {r=3}), B=Constant() {r=1.75})
  n11 = Constant3Vector() {constant=<rgba -94848 84352 0 1>}
  n12 = Constant() {r=89600}
  n13 = Constant() {r=0.75}
  n14 = WorldPosition()
  n15 = Divide(
   A=Add(A=ComponentMask(None=@n14.XYZ) {b=true}, B=Constant() {desc=SnowBlendHeightOffset, r=-5500})
   B=Constant() {desc=SnowBlendDistance, r=4000}
  )
  n16 = StaticSwitch(
   True=FunctionInput() {bUsePreviewValueAsDefault=true, inputName=SnowMask, inputType=FunctionInput_Scalar, sortPriority=4}
   False=Multiply(A=SphereMask(A=@n1.XYZ, B=@n11, Radius=@n12, Hardness=@n13) {hardnessPercent=0}, B=Saturate(None=@n15))
   Value=FunctionInput(Preview=StaticBool()) {bUsePreviewValueAsDefault=true, inputName=UseManualSnowMask, inputType=FunctionInput_StaticBool, sortPriority=3}
  )
  n17 = MaterialFunctionCall(Transition Phase=@n16, Height Texture=Multiply(A=@n6.A, B=Saturate(None=@n10))) {materialFunction=&HeightLerp}
  n18 = OneMinus(None=FunctionInput() {bUsePreviewValueAsDefault=true, inputName=Heightmap, inputType=FunctionInput_Scalar})
  n19 = Add(A=Constant() {r=-0.75}, B=Multiply(A=Constant() {r=2}, B=@n9))
  FunctionOutput(None=Multiply(
    A=MaterialFunctionCall(Transition Phase=@n17.Lerp Alpha No Contrast, Height Texture=@n18) {materialFunction=&HeightLerp}.Lerp Alpha No Contrast
    B=Saturate(None=@n19)
   )) {outputName=MaterialMask, sortPriority=9}
  n20 = Constant() {r=1.5}
  FunctionOutput(None=Saturate(None=Multiply(A=@n16, B=@n20))) {outputName=GrassMask, sortPriority=10}
  n21 = Abs(None=ComponentMask(None=VertexNormalWS()) {b=true})
  n22 = Constant() {r=1.25}
  n23 = Saturate(None=Add(A=Multiply(A=@n21, B=@n22), B=Constant() {r=-0.5}))
  n24 = Constant() {r=0.35}
  n25 = SetMaterialAttributes!(
   MaterialAttributes
   BaseColor=Multiply(A=Max(A=@n23, B=@n24), B=ScalarParameter() {defaultValue=0.6, group=ProceduralSnow, parameterName=SnowAlbedo-Mobile})
   Specular=ScalarParameter() {defaultValue=0.75, group=ProceduralSnow, parameterName=SnowSpec-Mobile}
   Roughness=Multiply(A=ScalarParameter() {defaultValue=0.65, group=ProceduralSnow, parameterName=SnowRough-Mobile}, B=@n8.R)
   Normal=VertexNormalWS()
  ) {attributeSetTypes [
    #1
    #4
    #2
    #3
   ]}
  FunctionOutput(None=@n25) {outputName=MobileMaterial}
  n26 = SetMaterialAttributes!(MaterialAttributes, BaseColor=@n2, Roughness=Constant() {r=0.8}, Normal=VertexNormalWS(), Specular=Constant() {r=0.7}) {attributeSetTypes [
    #1
    #2
    #3
    #4
   ]}
  FunctionOutput(None=@n26) {outputName=CheapMaterial, sortPriority=8}
 ]}
