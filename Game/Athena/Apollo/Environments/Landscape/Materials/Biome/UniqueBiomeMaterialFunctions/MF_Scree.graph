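# Landscape material function that builds the scree (loose rock) layer for Apollo biome terrain.
# Scree textures and a finer rock detail texture, both on landscape coordinates, combine into tinted base color darkened by MF_HeightBasedDarkenCliff, plus an angle-corrected blended normal.
# Specular multiplies a mid-range band of detail red with capped scree mask red, while roughness uses mask blue.
# Diffuse alpha feeds both Height and ConstantMask outputs; LayerWeight is unconnected.
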MaterialFunction /Game/Athena/Apollo/Environments/Landscape/Materials/Biome/UniqueBiomeMaterialFunctions/MF_Scree
graph {nodes [
  Constant() {r=0.2}
  FunctionInput() {inputName=LayerWeight, inputType=FunctionInput_Scalar}
  n1 = Multiply(A=LandscapeLayerCoords() {mappingScale=1}, B=Constant() {r=0.08})
  n2 = TextureSampleParameter2D(UVs=@n1) {group=Scree, parameterName=Scree_Diffuse, texture=/Game/Environments/Apollo/Sets/Terrain/Textures/T_Apollo_Scree_D_01}
  n3 = Divide(A=@n2.RGB, B=Constant3Vector() {constant=<rgba 1.04 1.2884 1.6 1>})
  n4 = Multiply(A=LandscapeLayerCoords() {mappingScale=1}, B=ScalarParameter() {defaultValue=0.2, parameterName=LS-ScreeDetailScale})
  n5 = TextureSample(UVs=@n4) {samplerSource=SSM_Wrap_WorldGroupSettings, texture=/Game/Environments/Apollo/Rocks/Natural_Rocks/Textures/T_RockDetails_02D}
  n6 = Multiply(A=Constant() {r=1.5}, B=Multiply(A=Saturate(None=@n3), B=SquareRoot(None=@n5.RGB)))
  n7 = Constant() {r=2.05}
  n8 = Saturate(None=Add(A=Constant() {r=-1.1}, B=Multiply(A=@n7, B=@n5.R)))
  n9 = Saturate(None=LinearInterpolate(A=Constant() {r=7}, B=Constant() {r=-0.8}, Alpha=@n5.R))
  n10 = TextureSampleParameter2D(UVs=@n1) {
   group=Scree
   parameterName=Scree_S
   samplerType=SAMPLERTYPE_Masks
   texture=/Game/Environments/Apollo/Sets/Terrain/Textures/T_Apollo_Scree_S_01
  }
  n11 = Saturate(None=LinearInterpolate(A=Constant() {r=-0.25}, B=Constant() {r=1.5}, Alpha=@n10.R))
  n12 = Constant() {r=0.5}
  n13 = TextureSampleParameter2D(UVs=@n1) {
   group=Scree
   parameterName=Scree_Normal
   samplerType=SAMPLERTYPE_Normal
   texture=/Game/Environments/Apollo/Sets/Terrain/Textures/T_Apollo_Scree_N_01
  }
  n14 = TextureSample(UVs=@n4) {
   samplerSource=SSM_Wrap_WorldGroupSettings
   samplerType=SAMPLERTYPE_Normal
   texture=/Game/Environments/Apollo/Rocks/Natural_Rocks/Textures/T_RockDetails_02_N
  }
  n15 = SetMaterialAttributes!(
   MaterialAttributes
   BaseColor=Multiply(
    A=MaterialFunctionCall!(DarknessAmount) {materialFunction=/Game/Athena/Apollo/Environments/Landscape/Materials/Biome/CommonMaterials/MF_HeightBasedDarkenCliff}.Result
    B=Saturate(None=@n6)
   )
   Specular=Multiply(A=Multiply(A=@n8, B=@n9), B=Clamp(None=@n11, Max=@n12) {clampMode=CMODE_ClampMax})
   Roughness=@n10.B
   Normal=MaterialFunctionCall(BaseNormal=@n13.RGB, AdditionalNormal=@n14.RGB) {materialFunction=&BlendAngleCorrectedNormals}.Result
  ) {attributeSetTypes [
    #1
    #2
    #3
    #4
   ]}
  FunctionOutput(None=@n15) {outputName=Material}
  FunctionOutput(None=@n2.A) {outputName=Height, sortPriority=1}
  FunctionOutput(None=@n2.A) {outputName=ConstantMask, sortPriority=1}
  Multiply(A=@n14.RGB, B=VectorParameter() {parameterName=LS-Scree-DetailNormal, defaultValue=<rgba 1 1 1 1>}.RGB)
  VectorParameter() {parameterName=MF_Scree-ColorDiv, defaultValue=<rgba 1.04 1.2884 1.6 1>}
 ]}
